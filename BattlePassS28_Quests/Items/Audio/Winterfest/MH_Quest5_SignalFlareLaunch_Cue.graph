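# Sound cue for the signal flare launch in a Season 28 Winterfest battle pass quest, routed to the quest sound class.
# A mixer layers a reward-screen chest upgrade sound at reduced volume with a briefly delayed, randomly chosen firework launch from three emote variants, modulated in pitch and volume.
# It borrows a supply-drop star attenuation, caps concurrent instances, and sends to a music analysis submix.
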SoundCue /BattlePassS28_Quests/Items/Audio/Winterfest/MH_Quest5_SignalFlareLaunch_Cue
root=Mixer_0
nodes [
 {id=Delay_0, children [Modulator_0], properties {delayMax=0.2, delayMin=0.2}}
 {id=Mixer_0, children [Delay_0,WavePlayer_0], properties {inputVolume [1,0.3]}}
 {id=Modulator_0, children [Random_0], properties {pitchMax=1.1, pitchMin=0.9, volumeMax=1, volumeMin=0.8}}
 {id=Random_0, children [WavePlayer_12, WavePlayer_13, WavePlayer_17], properties {bShouldExcludeFromBranchCulling=true, weights [1,1,1]}}
 {id=WavePlayer_0, properties {soundWaveAssetPtr=&/Fort_RewardScreen_TeamScore_ChestLootUpgrade}}
 {id=WavePlayer_12, properties {soundWaveAssetPtr=/Game/Athena/Sounds/Emotes/CandyApple_Fireworks/CA_Emote_Fireworks_Launch_03}}
 {id=WavePlayer_13, properties {soundWaveAssetPtr=/Game/Athena/Sounds/Emotes/CandyApple_Fireworks/CA_Emote_Fireworks_Launch_04}}
 {id=WavePlayer_17, properties {soundWaveAssetPtr=/Game/Athena/Sounds/Emotes/CandyApple_Fireworks/CA_Emote_Fireworks_Launch_08}}
]
properties {
 attenuationSettings=/Game/Athena/Sounds/SupplyDrops/Fort_HiddenStar_Ambient_Loop_Att
 bOverrideConcurrency=true
 maxDistance=2500
 soundClassObject=&Fort_Quests
 volumeMultiplier=1
 concurrencyOverrides {maxCount=3}
 soundSubmixSends [{minSendLevel=0, soundSubmix=&Fortnite_AudioAnalysis_Music}]
}
